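; Open Desert flame burp effect: drives one Niagara component at the positions of assigned building actors that are not destroyed.
; The server polls World Partition streaming on a timer after a deliberate initial wait, then binds each actor's OnDied to rebuild the position list.
; Component-local positions replicate through an OnRep array, which feeds Niagara's positions and loop divisor and deactivates it when empty.
; The construction script also rebuilds and applies the list.
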(event EventBeginPlay
  (if (Networking|IsServer)
    (if (Editor|InEditororPIE)
      (Utilities|FlowControl|Delay 5.0))
    (Default|SetStreamingTimer) ; We must wait because begin play happens before WP starts streaming so it returns as True for streaming being complete
  ))

(event Custom|CheckForStreamingComplete
  (Utilities|Time|ClearandInvalidateTimerbyHandle (Variables|Default|GettimerHandle_CheckForComplete))
  (bind _returnvalue (Streaming|IsAllStreamingCompleted (WorldSubsystems|GetWorldPartitionSubsystem)))
  (if _returnvalue
    (CallFunction|SignUpAttachement)
    (else
      (Default|SetStreamingTimer))))

(event Custom|SignUpAttachement
  (if (Networking|IsServer)
    (for _array_element (Variables|Default|GetAttachToActors)
      (if (Utilities|IsValidSoftObjectReference _array_element)
        (Utilities|IsValid (Utilities|ResolveSoftReference _array_element)
          (:Is%20Valid
            (EventDispatchers|BindEventtoOnDied (Utilities|ResolveSoftReference _array_element) (delegate Custom|HandleAttacheeDeath))))))
    (CallFunction|ValidatePositions)))

(event Custom|HandleAttacheeDeath
  (params (DamagedActor :type Actor%20Object%20Reference) (Damage :type Float%20%28single-precision%29) (InstigatedBy :type Controller%20Object%20Reference) (DamageCauser :type Actor%20Object%20Reference) (HitLocation :type Vector) (FHitComponent :type Primitive%20Component%20Object%20Reference) (BoneName :type Name) (Momentum :type Vector))
  (CallFunction|ValidatePositions))

(fn AddActorToArray
  (params (Actor :type Actor%20Object%20Reference))
  (bind _asbuilding_static_mesh (Utilities|Casting|CastToBuildingSMActor Actor))
  (if (not (Variables|/Script/FortniteGame.BuildingActor|GetbDestroyed _asbuilding_static_mesh))
    (Utilities|Array|Add (Variables|Default|GetVector%20Array%20Server) (Math|Transform|InverseTransformLocation (Transformation|GetWorldTransform (Variables|Default|GetNiagara)) (Transformation|GetActorLocation Actor)))))

(fn OnRep_VectorArrayRep
  (CallFunction|UpdateNiagara))

(fn SetStreamingTimer
  (if (not (Utilities|Time|IsValidTimerHandle (Variables|Default|GettimerHandle_CheckForComplete)))
    (bind _returnvalue (Utilities|Time|SetTimerbyEvent (EventDispatchers|CreateEvent :!event CheckForStreamingComplete) 5.0 false true))
    (Variables|Default|SettimerHandle_CheckForComplete _returnvalue)))

(fn UpdateNiagara
  (bind _vector_array_rep (Variables|Default|GetVector%20Array%20Rep))
  (bind _niagara (Variables|Default|GetNiagara))
  (Niagara|NiagaraSetPositionArray _niagara "Positions" (Variables|Default|GetVector%20Array%20Rep))
  (Components|Activation|SetActive (Variables|Default|GetNiagara) (Utilities|Array|IsValidIndex _vector_array_rep) false)
  (Effects|Components|ParticleSystem|SetFloatParameter _niagara "Loop Divisor" (Math|Conversions|ToFloat(Integer) (Utilities|Array|Length _vector_array_rep))))

(fn ConstructionScript
  (CallFunction|ValidatePositions)
  (CallFunction|UpdateNiagara))

(fn ValidatePositions
  (bind _vector_array_server (Variables|Default|GetVector%20Array%20Server))
  (Utilities|Array|Clear _vector_array_server)
  (for _array_element (Variables|Default|GetAttachToActors)
    (if (Utilities|IsValidSoftObjectReference _array_element)
      (Utilities|IsValid (Utilities|ResolveSoftReference _array_element)
        (:Is%20Valid
          (CallFunction|AddActortoArray (Utilities|ResolveSoftReference _array_element))))))
  (Variables|Default|SetVector%20Array%20Rep _vector_array_server))
