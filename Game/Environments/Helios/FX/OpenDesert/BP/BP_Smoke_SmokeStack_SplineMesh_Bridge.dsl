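; Smoke stack spline mesh bridge Blueprint for the Helios open desert environment FX.
; Its construction script sets the "Time Offset" scalar parameter on the static mesh's materials from a Time Offset variable, so the smoke material's timing can vary per instance.
; BeginPlay contains no logic.
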(event EventBeginPlay)

(fn ConstructionScript
  (Rendering|Material|SetScalarParameterValueonMaterials (Variables|Default|GetStaticMesh) "Time Offset" (Variables|Default|GetTime%20Offset)))
