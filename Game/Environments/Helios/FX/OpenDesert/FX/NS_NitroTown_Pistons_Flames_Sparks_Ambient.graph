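# Niagara system for ambient glowing sparks at the NitroTown pistons in the Helios open desert environment.
# GPU Embers emitter partitions particles across eight fixed offsets, with spawn rate fading over each loop.
# Upward gravity, curl noise and drag drive motion, while waveforms flicker color and size, scaled by camera distance for two sprite renderers.
# Disabled EmbersDragPiston variant adds a growing cylinder spawn and negative-strength point attraction at the same offsets.
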NiagaraSystem /Game/Environments/Helios/FX/OpenDesert/FX/NS_NitroTown_Pistons_Flames_Sparks_Ambient
dependencies {
 usedDataInterfaces [&NiagaraDataInterfaceCurve, &NiagaraDataInterfaceArrayFloat3]
 usedDynamicInputs []
 usedModules [&SystemState]
}
emitters {
 Embers {
  simTarget=GPUComputeSim
  emitterUpdateScript [
   &EmitterState()
   &SpawnRate(
    SpawnRate=&FloatFromCurve(FloatCurve=DataInterface() {propertyValues {Curve {keys [<at 0 1>,<at 1 0>]}}}, CurveIndex=$Emitter.NormalizedLoopAge, Scale Curve=250)
    Use Spawn Probability=false
   )
  ]
  particleSpawnScript [
   &InitializeParticle(
    Color=&RandomRangeLinearColor(Minimum=<rgba 25 10.323 0 1>, Maximum=<rgba 250 129.0156 0 1>)
    Lifetime Max=3
    Lifetime Mode=::NewEnumerator1
    Sprite Rotation Mode=::NewEnumerator2
    Sprite Size Mode=::NewEnumerator1
    Uniform Sprite Size Max=100
    Uniform Sprite Size Min=25
    UsePositionOffset=false
    Write Material Random=true
   )
   &PartitionParticles(Number Of Partitions=8)
   &AddVelocity(
    Cone Angle=80
    Cone Axis=<0 0 1>
    Inner Cone Angle=50
    Module.UseConeFalloff=false
    Module.UseVelDistribution=false
    Override Seed=false
    Velocity Mode=::NewEnumerator2
    Velocity Speed=250
   )
   &ShapeLocation(
    MaskSpawnGroup=false
    Offset=&SelectVectorFromArray(
     Array Sampling Mode=::NewEnumerator0
     Direct Array Index=$Particles.PartitionParticles.Partition
     Vector Selection Array=DataInterface() {propertyValues {FloatData [
        <-963.6505 -998.7991 0>
        <-531.2061 -998.7991 0>
        <512.9941 -998.7991 0>
        <977.9363 -998.7991 0>
        <977.9363 1023.2148 0>
        <502.1314 1023.2148 0>
        <-534.7775 1023.2148 0>
        <-1046.9307 1023.2148 0>
       ]}}
     Random Seed=0
    )
    Offset Mode=::NewEnumerator0
    Override Seed=false
    UseOwnerScale=true
   )
  ]
  particleUpdateScript [
   /SetVariables(Particles.VisibilityTag=&SpawnGroup()) {enabled=false, isSetParameters=true, module=SetVariables_#1}
   &GravityForce(Gravity=<0 0 250>)
   &CurlNoiseForce(
    Noise Frequency=&RandomRangeFloat(Minimum=1, Maximum=5)
    Noise Strength=&FloatFromCurve(
     FloatCurve=DataInterface() {propertyValues {Curve {keys [<at 0 1>,<at 0.5 0.125>,<at 0.9 1.75>,<at 1.3 2>]}}}
     Scale Curve=&RandomRangeFloat(Minimum=250, Maximum=500)
    )
    Pan Noise=true
   )
   &Drag(
    Drag=&FloatFromCurve(FloatCurve=DataInterface() {propertyValues {Curve {keys [<at 0 0>,<at 0.4 1>,<at 1 0>]}}}, Scale Curve=2)
    Use Linear Drag=true
    Use Rotational Drag=false
   )
   &SolveForcesAndVelocity(Acceleration Limit=100, Clamp Velocity=true, Limit Acceleration=true, Speed Limit=150)
   &ParticleState()
   &DynamicMaterialParameters(
    Index 0 Param 1=0.8
    Index 0 Param 2=1
    Index 0 Param 3=1
    Index 1 Param 1=150
    Index 1 Param 4=1
    Module.Param0WriteEnabled=true
    Module.Param0WriteEnabled1=true
    Module.Param1WriteEnabled=true
    Module.Param1WriteEnabled1=true
    Module.Param2WriteEnabled=true
    Module.Param2WriteEnabled1=true
    Module.Param3WriteEnabled=false
    Module.Param3WriteEnabled1=true
    Write Parameter Index 1=true
   )
   &ScaleColor(
    Scale RGB=&Multiply_VectorByFloat(
     Vector=<25 5 0.05>
     Float=&FloatFromCurve(
      FloatCurve=DataInterface() {propertyValues {Curve {keys [{interpMode=RCIM_Cubic, tangentMode=RCTM_User, time=0, value=1}, {interpMode=RCIM_Cubic, tangentMode=RCTM_User, time=1, value=0}]}}}
      Scale Curve=&Multiply_Float(
       A=&Multiply_Float(
        A=1
        B=&Waveform(
         "Waveform [1]"=::NewEnumerator0
         "[1] Phase Scale"=&RandomRangeFloat(Minimum=1, Maximum=2)
         "[1] Frequency"=&RandomRangeFloat(Minimum=2, Maximum=2)
         "[1] Amplitude Min/Max"=<0.5 1>
        )
       )
       B=&FN_LOD_Distance_Remap(Start Distance=2500, End Distance=5000, Minimum Output=0.5, Maximum Output=1, Clamp Output=true)
      )
     )
    )
    ScaleA=true
    ScaleRGB=true
   ) {module=ScaleColor001}
   &ScaleColor(Scale Alpha=0.02, Scale RGB=<0.02 0.01 0.01>, ScaleA=true, ScaleRGB=true) {enabled=false}
   /Niagara/Modules/Update/Camera/MaintainInCameraParticleScale(Pixel Depth=0.1, Render Target Width=0.1)
   &ScaleSpriteSizeBySpeed(
    Max Scale Factor=<0.15 0.3>
    Min Scale Factor=<0.1 0.15>
    Sample Scale Factor By Curve=true
    Scale Factor Curve=DataInterface() {propertyValues {Curve {keys [
        {interpMode=RCIM_Cubic, tangentMode=RCTM_User, time=0, value=0}
        {arriveTangent=2.9664, interpMode=RCIM_Cubic, leaveTangent=2.9664, tangentMode=RCTM_User, time=1, value=1}
       ]}}}
    Velocity Threshold=150
   )
   &ScaleSpriteSize(
    Uniform Curve Scale=&Multiply_Float(
     A=&Multiply_Float(
      A=0.2
      B=&Waveform(
       "Waveform [1]"=::NewEnumerator0
       "[1] Phase Scale"=&RandomRangeFloat(Minimum=3, Maximum=10)
       "[1] Frequency"=&RandomRangeFloat(Minimum=0.2, Maximum=0.5)
       "[1] Amplitude Min/Max"=<0.9 1>
      )
     )
     B=&FN_LOD_Distance_Remap(Start Distance=2500, End Distance=5000, Minimum Output=0.5, Maximum Output=1, Clamp Output=true)
    )
    Uniform Curve Sprite Scale=DataInterface() {propertyValues {Curve {keys [
        {arriveTangent=2.9868, interpMode=RCIM_Cubic, leaveTangent=2.9868, tangentMode=RCTM_User, time=0, value=1.65}
        <at 0.4 2.3>
        {arriveTangent=-0.0405, interpMode=RCIM_Cubic, leaveTangent=-0.0405, tangentMode=RCTM_User, time=1, value=1.4}
       ]}}}
   )
   /SetVariables(
    Particles.Linear Color=&MultiplyLinearColor(Linear Color=$Particles.Color, Scale Factor=<0.002 0.002 0.002 0.01>)
    Particles.Vector:Vector3f=&Multiply_Vector(A=$Particles.Velocity, B=<1 1 1>)
    Particles.Vector 2D:Vector2f=&Multiply_Vector2D(A=<2 2>, B=$Particles.SpriteSize)
   ) {isSetParameters=true, module=SetVariables_#2}
  ]
  rendererClasses [&NiagaraSpriteRendererProperties]
  renderers [
   {rendererClass=&NiagaraSpriteRendererProperties}
   {rendererClass=&NiagaraSpriteRendererProperties, rendererIndex=1}
  ]
 }
 EmbersDragPiston {
  bEnabled=false
  simTarget=GPUComputeSim
  emitterUpdateScript [
   &EmitterState()
   &SpawnRate(
    SpawnRate=&FloatFromCurve(FloatCurve=DataInterface() {propertyValues {Curve {keys [<at 0 1>,<at 1 0>]}}}, CurveIndex=$Emitter.NormalizedLoopAge, Scale Curve=2500)
    Use Spawn Probability=false
   )
  ]
  particleSpawnScript [
   &InitializeParticle(
    Color=&RandomRangeLinearColor(Minimum=<rgba 25 10.323 0 1>, Maximum=<rgba 250 129.0156 0 1>)
    Lifetime Max=3
    Lifetime Mode=::NewEnumerator1
    Sprite Rotation Mode=::NewEnumerator2
    Sprite Size Mode=::NewEnumerator1
    Uniform Sprite Size Max=100
    Uniform Sprite Size Min=25
    UsePositionOffset=false
    Write Material Random=true
   )
   &PartitionParticles(Number Of Partitions=8, Particles Per Partition=1000, Partition Mode:ENiagara_PartitionMode=::NewEnumerator1)
   &AddVelocity(
    Cone Angle=25
    Cone Axis=<0 0 1>
    Inner Cone Angle=15
    Module.UseConeFalloff=false
    Module.UseVelDistribution=false
    Override Seed=false
    Velocity Mode=::NewEnumerator2
    Velocity Speed=500
   )
   &ShapeLocation(
    Cylinder Height=&FloatFromCurve(FloatCurve=DataInterface() {propertyValues {Curve {keys [<at 0 0>,<at 1 1>]}}}, CurveIndex=$Emitter.NormalizedLoopAge, Scale Curve=1500)
    Cylinder Height Midpoint=0
    MaskSpawnGroup=false
    Offset=&SelectVectorFromArray(
     Direct Array Index=$Particles.PartitionParticles.Partition
     Vector Selection Array=DataInterface() {propertyValues {FloatData [
        <-963.6505 -998.7991 0>
        <-531.2061 -998.7991 0>
        <512.9941 -998.7991 0>
        <977.9363 -998.7991 0>
        <977.9363 1023.2148 0>
        <502.1314 1023.2148 0>
        <-534.7775 1023.2148 0>
        <-1046.9307 1023.2148 0>
       ]}}
    )
    Offset Mode=::NewEnumerator0
    Override Seed=false
    Shape Primitive=::NewEnumerator1
    Surface Only=true
    UseOwnerScale=true
   )
  ]
  particleUpdateScript [
   /SetVariables(Particles.VisibilityTag=&SpawnGroup()) {enabled=false, isSetParameters=true, module=SetVariables_#1}
   &GravityForce(Gravity=<0 0 250>)
   &CurlNoiseForce(
    Curl Noise Cone Mask Angle=80
    Curl Noise Cone Mask Falloff Angle=60
    Mask Curl Noise=true
    Noise Frequency=&RandomRangeFloat(Minimum=2, Maximum=5)
    Noise Strength=&FloatFromCurve(
     FloatCurve=DataInterface() {propertyValues {Curve {keys [<at 0 2>,<at 1 0>]}}}
     Scale Curve=&RandomRangeFloat(Minimum=500, Maximum=1500)
    )
    Pan Noise=true
   )
   &Drag(
    Drag=&FloatFromCurve(FloatCurve=DataInterface() {propertyValues {Curve {keys [<at 0 0>,<at 0.4 1>,<at 1 0>]}}}, Scale Curve=25)
    Use Linear Drag=true
    Use Rotational Drag=false
   )
   &PointAttractionForce(
    Attraction Radius=500
    Attraction Strength=-25
    Attractor Position Offset:Vector3f=&SelectVectorFromArray(
     Direct Array Index=$Particles.PartitionParticles.Partition
     Vector Selection Array=DataInterface() {propertyValues {FloatData [
        <-963.6505 -998.7991 0>
        <-531.2061 -998.7991 0>
        <512.9941 -998.7991 0>
        <977.9363 -998.7991 0>
        <977.9363 1023.2148 0>
        <502.1314 1023.2148 0>
        <-534.7775 1023.2148 0>
        <-1046.9307 1023.2148 0>
       ]}}
    )
    Kill Within Radius=false
    Module.Use Falloff=true
    UseAttractorOffset=true
   )
   &SolveForcesAndVelocity(Acceleration Limit=500, Clamp Velocity=true, Limit Acceleration=true, Speed Limit=500)
   &ParticleState()
   &DynamicMaterialParameters(
    Index 0 Param 1=0.8
    Index 0 Param 2=1
    Index 0 Param 3=1
    Index 1 Param 1=150
    Index 1 Param 4=1
    Module.Param0WriteEnabled=true
    Module.Param0WriteEnabled1=true
    Module.Param1WriteEnabled=true
    Module.Param1WriteEnabled1=true
    Module.Param2WriteEnabled=true
    Module.Param2WriteEnabled1=true
    Module.Param3WriteEnabled=false
    Module.Param3WriteEnabled1=true
    Write Parameter Index 1=true
   )
   &ScaleColor(
    Scale RGB=&Multiply_VectorByFloat(
     Vector=<25 5 0.05>
     Float=&FloatFromCurve(
      FloatCurve=DataInterface() {propertyValues {Curve {keys [{interpMode=RCIM_Cubic, tangentMode=RCTM_User, time=0, value=1}, {interpMode=RCIM_Cubic, tangentMode=RCTM_User, time=1, value=0}]}}}
      Scale Curve=&Multiply_Float(
       A=&Multiply_Float(
        A=1
        B=&Waveform(
         "Waveform [1]"=::NewEnumerator0
         "[1] Phase Scale"=&RandomRangeFloat(Minimum=1, Maximum=2)
         "[1] Frequency"=&RandomRangeFloat(Minimum=2, Maximum=2)
         "[1] Amplitude Min/Max"=<0.5 1>
        )
       )
       B=&FN_LOD_Distance_Remap(Start Distance=2500, End Distance=5000, Minimum Output=0.5, Maximum Output=1, Clamp Output=true)
      )
     )
    )
    ScaleA=true
    ScaleRGB=true
   ) {module=ScaleColor001}
   &ScaleColor(Scale Alpha=0.02, Scale RGB=<0.02 0.01 0.01>, ScaleA=true, ScaleRGB=true) {enabled=false}
   &MaintainInCameraParticleScale(Pixel Depth=0.1, Render Target Width=0.1)
   &ScaleSpriteSizeBySpeed(
    Max Scale Factor=<0.15 0.3>
    Min Scale Factor=<0.1 0.15>
    Sample Scale Factor By Curve=true
    Scale Factor Curve=DataInterface() {propertyValues {Curve {keys [
        {interpMode=RCIM_Cubic, tangentMode=RCTM_User, time=0, value=0}
        {arriveTangent=2.9664, interpMode=RCIM_Cubic, leaveTangent=2.9664, tangentMode=RCTM_User, time=1, value=1}
       ]}}}
    Velocity Threshold=150
   )
   &ScaleSpriteSize(
    Uniform Curve Scale=&Multiply_Float(
     A=&Multiply_Float(
      A=0.2
      B=&Waveform(
       "Waveform [1]"=::NewEnumerator0
       "[1] Phase Scale"=&RandomRangeFloat(Minimum=3, Maximum=10)
       "[1] Frequency"=&RandomRangeFloat(Minimum=0.2, Maximum=0.5)
       "[1] Amplitude Min/Max"=<0.9 1>
      )
     )
     B=&FN_LOD_Distance_Remap(Start Distance=2500, End Distance=5000, Minimum Output=0.5, Maximum Output=1, Clamp Output=true)
    )
    Uniform Curve Sprite Scale=DataInterface() {propertyValues {Curve {keys [
        {arriveTangent=2.9868, interpMode=RCIM_Cubic, leaveTangent=2.9868, tangentMode=RCTM_User, time=0, value=1.65}
        <at 0.4 2.3>
        {arriveTangent=-0.0405, interpMode=RCIM_Cubic, leaveTangent=-0.0405, tangentMode=RCTM_User, time=1, value=1.4}
       ]}}}
   )
   /SetVariables(
    Particles.Linear Color=&MultiplyLinearColor(Linear Color=$Particles.Color, Scale Factor=<0.002 0.002 0.002 0.01>)
    Particles.Vector:Vector3f=&Multiply_Vector(A=$Particles.Velocity, B=<1 1 1>)
    Particles.Vector 2D:Vector2f=&Multiply_Vector2D(A=<2 2>, B=$Particles.SpriteSize)
   ) {isSetParameters=true, module=SetVariables_#2}
  ]
  rendererClasses [&NiagaraSpriteRendererProperties]
  renderers [
   {rendererClass=&NiagaraSpriteRendererProperties}
   {rendererClass=&NiagaraSpriteRendererProperties, rendererIndex=1}
  ]
 }
}
summary {
 systemName=NS_NitroTown_Pistons_Flames_Sparks_Ambient
 emitters [
  {emitterName=Embers, simTarget=GPUComputeSim, rendererClasses [&NiagaraSpriteRendererProperties]}
  {bEnabled=false, emitterName=EmbersDragPiston, simTarget=GPUComputeSim, rendererClasses [&NiagaraSpriteRendererProperties]}
 ]
 userVariables []
}
