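# Masked Niagara mesh-particle material for looping sparks, animated entirely in shader math from time.
# Vertices cycle along OverallDirection with UV1 phase offsets, pass through Sprite_Capsule for shaping, and blend toward the object pivot via a sphere mask.
# Dynamic parameters control spark size, loop timing, speed and radii.
# Vertex-color green varies mask and hue shift on particle-colored emissive, and feeds a resolution-aware "Constant Screen size" reroute.
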Material /Game/Environments/Helios/FX/OpenDesert/Materials/M_MeshParticle_NoTick_Sparks_HotDog
graph {
 properties {
  bCastRayTracedShadows=false
  bTangentSpaceNormal=false
  bUsedWithNiagaraMeshParticles=true
  blendMode=BLEND_Masked
  pixelDepthOffsetMode=PDOM_Legacy
  usedShadingModels=Default Lit
 }
 nodes [
  n1 = Constant()
  n2 = Multiply(
   A=ObjectRadius()
   B=Multiply(
    A=DynamicParameter() {defaultValue=<rgba 20 150 5 1>, paramNames [Spark_Width,Spark_Length,Loop,Radius]}.Radius
    B=ScalarParameter() {defaultValue=1, parameterName=ObjectRad}
   )
  ) {constB=0.5}
  n3 = Divide(A=Time(), B=DynamicParameter() {defaultValue=<rgba 20 150 5 1>, paramNames [Spark_Width,Spark_Length,Loop,Radius]}.Loop)
  n4 = ComponentMask(None=TextureCoordinate() {coordinateIndex=1}) {g=true}
  n5 = Add(
   A=Multiply(
    A=Frac(None=@n3)
    B=DynamicParameter() {parameterIndex=1, defaultValue=<rgba 5000 1 1 1>, paramNames [Radius,MultSpeed,Param3,Param4]}.MultSpeed
   )
   B=Multiply(A=Subtract(A=@n4) {constB=0.5}) {constB=-2}
  )
  n6 = Multiply(
   A=LinearInterpolate(A=Multiply(A=@n2) {constB=-1}, B=@n2, Alpha=Frac(None=@n5))
   B=VectorParameter() {parameterName=OverallDirection, defaultValue=<rgba 0 0 1 1>}.RGB
  )
  n7 = DynamicParameter() {defaultValue=<rgba 20 150 5 1>, paramNames [Spark_Width,Spark_Length,Loop,Radius]}
  n8 = MaterialFunctionCall!(
   Particle Position=ParticlePositionWS()
   Normalized Particle Direction=VectorParameter() {parameterName=Direction, defaultValue=<rgba 0 0 1 1>}.RGB
   Capsule Diameter=@n7.Spark_Width
   Capsule Pivot Alignment=ScalarParameter() {defaultValue=1, parameterName=Capsule Pivot Alignment}
   Scale Along Velocity Vector=@n7.Spark_Width
   ---------Scale Over Velocity Parameters---------
   Particle Speed
   Minimum Speed
   Max Speed
   Minimum Scale Factor
   Maximum Scale Factor
   ---------- Advanced -----------
   Cap Size Multiplier
   Use Camera Offset
   Use Advanced Features=StaticBool() {value=true}
   World Position to Negate
  ) {materialFunction=/Engine/Functions/Engine_MaterialFunctions02/WorldPositionOffset/Sprite_Capsule}
  n9 = ObjectPositionWS()
  n10 = WorldPosition()
  n11 = SphereMask(
   A=@n9
   B=@n10.XYZ
   Radius=DynamicParameter() {parameterIndex=1, defaultValue=<rgba 5000 1 1 1>, paramNames [Radius,MultSpeed,Param3,Param4]}.Radius
  ) {hardnessPercent=0}
  Add(A=Subtract(A=Add(A=@n1, B=@n6), B=@n8.World Position Offset), B=LinearInterpolate(B=Subtract(A=@n9, B=@n10.XYZ), Alpha=Saturate(None=@n11)))
  Constant() {r=0.75}
  n12 = VertexColor()
  n13 = ScalarParameter() {defaultValue=0.85, parameterName=Paramasd}
  n14 = ParticleColor()
  Multiply(A=Power(Base=Multiply(A=@n12.G, B=@n13)), B=Multiply(A=@n8.Opacity Mask, B=@n14.A))
  n15 = LinearInterpolate(Alpha=VertexColor().G) {constB=0.15}
  Multiply(
   A=MaterialFunctionCall(Emissive=MaterialFunctionCall(Hue Shift Percentage=@n15, Texture=@n14.RGB) {materialFunction=&HueShift}.Result) {materialFunction=&MF_FX_EmissiveAdjustments}.Result
   B=ScalarParameter() {defaultValue=1, parameterName=Emi}
  )
  n16 = Multiply(A=ScalarParameter() {defaultValue=0.25, group=WPO, parameterName=MinSizeMult}, B=VertexColor().G)
  n17 = ScalarParameter() {defaultValue=25, group=WPO, parameterName=maxSize}
  n18 = PixelDepth()
  n19 = MaterialFunctionCall() {materialFunction=&ScreenResolution}
  n20 = Divide(A=ComponentMask(None=@n19.Visible Resolution) {r=true}, B=ComponentMask(None=@n19.Visible Resolution) {g=true})
  n21 = ScalarParameter() {defaultValue=1.33, parameterName=ScreenRes_Mininum}
  n22 = ScalarParameter() {defaultValue=2, parameterName=ScreenRes_Max}
  n23 = Divide(A=Subtract(A=Clamp(None=@n20, Min=@n21, Max=@n22) {maxDefault=2, minDefault=1.33}, B=@n21), B=Subtract(A=@n22, B=@n21))
  n24 = Multiply(
   A=ComponentMask(None=ViewSize()) {g=true}
   B=LinearInterpolate(
    A=ScalarParameter() {defaultValue=0.08, group=WPO, parameterName=MinSizeatDistance_43_ScreenRatio}
    B=ScalarParameter() {defaultValue=0.08, group=WPO, parameterName=MinSizeatDistance_Full_ScreenRatio}
    Alpha=Saturate(None=@n23)
   )
  )
  NamedRerouteDeclaration(None=Add(A=Multiply(A=@n16, B=@n17), B=Divide(A=@n18.R, B=@n24))) {name=Constant Screen size, nodeColor=<rgba 0.9922 1 0 1>}
 ]
}
