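# Translucent, unlit, two-sided material for NitroTown flame bursts in the Helios open desert, named as an HLOD variant.
# A looping, offsettable timer drives vertical displacement weighted along V, added to SplineThicken widening, and fades each burst out.
# Panning plasma-flame noise shapes an edge-faded alpha, while an orange UV gradient passes through DeriveHDRfromLDR for glow.
# Pixel depth scales both displacement and alpha, so flames diminish near the camera.
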Material /Game/Environments/Helios/FX/OpenDesert/Materials/M_NitroTown_FlameBurps_Hlod
graph {
 properties {
  bAutomaticallySetUsageInEditor=false
  blendMode=BLEND_Translucent
  pixelDepthOffsetMode=PDOM_Legacy
  shadingModel=MSM_Unlit
  twoSided=true
  usedShadingModels=Unlit
 }
 nodes [
  LinearInterpolate()
  n1 = OneMinus(None=ComponentMask(None=TextureCoordinate()) {g=true})
  LinearInterpolate(Alpha=@n1)
  n2 = TextureCoordinate()
  n3 = LinearInterpolate(
   A=VectorParameter() {parameterName=Param_3, defaultValue=<rgba 1 0.0828 0 1>}.RGB
   B=VectorParameter() {parameterName=Param_2, defaultValue=<rgba 5 1.4091 0 1>}.RGB
   Alpha=Power(Base=ComponentMask(None=@n2) {g=true}, Exp=ScalarParameter() {defaultValue=8, parameterName=Param_4})
  )
  MaterialFunctionCall(LDR Input=Saturate(None=@n3), HDR Tint=@n3, HDR Intensity Multiplier=ScalarParameter() {defaultValue=10, parameterName=Param_5}) {materialFunction=&DeriveHDRfromLDR}
  n4 = ScalarParameter() {defaultValue=200}
  n5 = ScalarParameter() {defaultValue=500, parameterName=Param_1}
  n6 = Time()
  n7 = ScalarParameter() {parameterName=Time Offset}
  n8 = LinearInterpolate(
   A=VectorParameter() {parameterName=Param_1, defaultValue=<rgba 0 0 -500 0>}.RGB
   B=VectorParameter() {defaultValue=<rgba 0 0 1250 0>}.RGB
   Alpha=Fmod(A=Add(A=@n6, B=@n7) {constB=0}, B=ScalarParameter() {defaultValue=1.5, parameterName=Param_2})
  )
  n9 = Saturate(None=Divide(A=PixelDepth().R, B=ScalarParameter() {defaultValue=35000, parameterName=Param_6}))
  Multiply(
   A=Add(
    A=MaterialFunctionCall(WidthBase=@n4, WidthTip=@n5) {materialFunction=&SplineThicken}.WorldPositionOffset
    B=Multiply(A=@n1, B=@n8)
   )
   B=@n9
  )
  n10 = TextureCoordinate()
  n11 = ComponentMask(None=@n10) {g=true}
  n12 = TextureSample(UVs=Panner() {speedY=1}) {samplerType=SAMPLERTYPE_LinearColor, texture=/Game/Items/Traps/Textures/T_Plasma_Flames_01_M}
  n13 = Multiply(A=@n12.G, B=ScalarParameter() {defaultValue=0.5, parameterName=Param_3})
  n14 = ComponentMask(None=@n10) {r=true}
  n15 = OneMinus(None=@n14)
  n16 = Multiply(A=Multiply(A=@n11, B=Add(A=@n11, B=@n13)), B=Multiply(A=Multiply(A=@n15, B=@n14)) {constB=4})
  n17 = Fmod(A=Add(A=Time(), B=ScalarParameter() {parameterName=Time Offset}) {constB=0}, B=ScalarParameter() {defaultValue=1.5, parameterName=Param_2})
  Multiply(A=Multiply(A=Multiply(A=@n16, B=@n12.G), B=OneMinus(None=@n17)), B=@n9)
 ]
}
