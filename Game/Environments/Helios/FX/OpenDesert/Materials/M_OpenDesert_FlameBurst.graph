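# Additive, unlit, two-sided flame burst material for Niagara mesh particles in the Helios OpenDesert environment.
# Distorted, panning smoke-stack textures, each sampled twice at offset UVs, build hue-shifted emissive, boosted for ray tracing, and opacity shaped by a vertical gradient, erosion and temporal dither.
# SplineThicken plus animated noise along vertex normals drives world position offset.
# Dynamic parameters supply width, tiling, speed, erosion and randomness per particle.
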Material /Game/Environments/Helios/FX/OpenDesert/Materials/M_OpenDesert_FlameBurst
graph {
 properties {
  bCastRayTracedShadows=false
  bUseTranslucencyVertexFog=false
  bUsedWithInstancedStaticMeshes=true
  bUsedWithNiagaraMeshParticles=true
  blendMode=BLEND_Additive
  pixelDepthOffsetMode=PDOM_Legacy
  shadingModel=MSM_Unlit
  twoSided=true
  usedShadingModels=Unlit
 }
 nodes [
  n1 = DynamicParameter() {defaultValue=<rgba 100 2500 1 1>, paramNames [WidthBase,WidthTip,Mult_Tile,Mult_Speed]}
  Add(
   A=NamedRerouteUsage()
   B=MaterialFunctionCall(WidthBase=@n1.WidthBase, WidthTip=@n1.WidthBase) {materialFunction=&SplineThicken}.WorldPositionOffset
  )
  n2 = Multiply(A=DynamicParameter() {parameterIndex=3, defaultValue=<rgba 1 1 2 2>, paramNames [Erode,Random,Param3,Param4]}.Erode) {constB=0.03}
  n3 = Add(A=AppendVector(A=Constant(), B=NamedRerouteUsage()), B=TextureCoordinate())
  n4 = DynamicParameter() {parameterIndex=2, defaultValue=<rgba 1 1 2 2>, paramNames [Flame_Tile_X, Flame_Tile_Y, Flame_Distort_Tile_X, Flame_Distort_Tile_Y]}
  n5 = Multiply(A=NamedRerouteUsage(), B=AppendVector(A=@n4.Flame_Tile_X, B=@n4.Flame_Tile_X))
  n6 = NamedRerouteUsage()
  n7 = DynamicParameter() {
   parameterIndex=1
   defaultValue=<rgba 1 1 2 2>
   paramNames [Flame_PannerSpeed_X, Flame_PannerSpeed_Y, Flame_Opacity_PannerSpeed_X, Flame_Opacity_PannerSpeed_Y]
  }
  n8 = AppendVector(A=@n7.Flame_PannerSpeed_X, B=@n7.Flame_PannerSpeed_X)
  n9 = Add(A=NamedRerouteUsage(), B=TextureCoordinate())
  n10 = DynamicParameter() {parameterIndex=2, defaultValue=<rgba 1 1 2 2>, paramNames [Flame_Tile_X, Flame_Tile_Y, Flame_Distort_Tile_X, Flame_Distort_Tile_Y]}
  n11 = Multiply(A=NamedRerouteUsage(), B=AppendVector(A=@n10.Flame_Distort_Tile_X, B=@n10.Flame_Distort_Tile_X))
  n12 = NamedRerouteUsage()
  n13 = AppendVector(
   A=ScalarParameter() {defaultValue=0.02, parameterName=FlameDistort_PannerSpeed_X}
   B=ScalarParameter() {defaultValue=0.02, parameterName=FlameDistort_PannerSpeed_Y}
  )
  n14 = TextureSampleParameter2D(UVs=Panner(Coordinate=Multiply(A=@n9, B=@n11), Speed=Multiply(A=@n12, B=@n13)) {speedX=-0.02, speedY=-0.05}) {
   parameterName=Texture_Distort
   samplerType=SAMPLERTYPE_LinearColor
   texture=/Game/Effects/Fort_Effects/Effects/Environment/Desert/Dust/Organize/T_Turbulence_Seamless
  }
  n15 = Multiply(A=ComponentMask(None=@n14.RGB) {g=true, r=true}, B=ScalarParameter() {defaultValue=0.05, parameterName=DistortMult})
  n16 = Add(A=Panner(Coordinate=Multiply(A=@n3, B=@n5), Speed=Multiply(A=@n6, B=@n8)) {speedX=0.05, speedY=0.15}, B=@n15)
  n17 = Add(A=@n16, B=Constant2Vector() {g=0.5})
  n18 = TextureSampleParameter2D(UVs=@n17) {parameterName=Texture_Erode, texture=/Game/Environments/Helios/FX/OpenDesert/FX/T_OpenDesert_SmokeStack_Distance_01}
  Add(A=@n2, B=@n18.G)
  AppendVector()
  Constant() {r=1}
  Constant() {r=1}
  DepthFade(FadeDistance=ScalarParameter() {defaultValue=100, parameterName=DepthFadeDistance})
  n19 = ParticleColor()
  n20 = VectorParameter() {parameterName=ColorTint, defaultValue=<rgba 0.6823 0.2697 0 1>}
  n21 = TextureSampleParameter2D(UVs=@n16) {parameterName=Texture_Color_01, texture=&T_OpenDesert_SmokeStack_Distance_01}
  n22 = TextureSampleParameter2D(UVs=@n17) {parameterName=Texture_Color_02, texture=&T_OpenDesert_SmokeStack_Distance_01}
  n23 = Add(A=LinearInterpolate(A=Constant(), Alpha=Add(A=@n21.G, B=@n22.G)), B=Constant())
  n24 = ScalarParameter() {defaultValue=1, parameterName=Brightness}
  n25 = Multiply(A=Multiply(
    A=MaterialFunctionCall(Texture=@n20.RGB) {materialFunction=&HueShift}.Result
    B=Multiply(A=@n23, B=@n24)
   ))
  n26 = Multiply(A=@n25, B=ScalarParameter() {defaultValue=3, parameterName=LumenBoost})
  MaterialFunctionCall(Emissive=Multiply(A=@n19.RGB, B=RayTracingQualitySwitch(Normal=@n25, RayTraced=@n26))) {materialFunction=&MF_FX_EmissiveAdjustments}
  n27 = MaterialFunctionCall() {materialFunction=&LinearGradient}
  n28 = Power(Base=Sine(None=Multiply(A=@n27.VGradient) {constB=0.5})) {constExponent=1}
  n29 = AppendVector(A=Constant(), B=NamedRerouteUsage())
  n30 = TextureCoordinate() {vTiling=0.5}
  n31 = NamedRerouteUsage()
  n32 = DynamicParameter() {
   parameterIndex=1
   defaultValue=<rgba 1 1 2 2>
   paramNames [Flame_PannerSpeed_X, Flame_PannerSpeed_Y, Flame_Opacity_PannerSpeed_X, Flame_Opacity_PannerSpeed_Y]
  }
  n33 = AppendVector(A=@n32.Flame_Opacity_PannerSpeed_X, B=@n32.Flame_Opacity_PannerSpeed_X)
  n34 = Add(A=Panner(Coordinate=Add(A=@n29, B=@n30), Speed=Multiply(A=@n31, B=@n33)) {speedY=0.2}, B=@n15)
  n35 = TextureSampleParameter2D(UVs=Add(A=@n34, B=Constant2Vector() {g=0.5})) {parameterName=Texture_Opactiy_02, texture=&T_OpenDesert_SmokeStack_Distance_01}
  n36 = TextureSampleParameter2D(UVs=@n34) {parameterName=Texture_Opactiy_01, texture=&T_OpenDesert_SmokeStack_Distance_01}
  n37 = Multiply(A=Saturate(None=Add(A=@n35.G, B=@n36.G)), B=ScalarParameter() {defaultValue=1.5, parameterName=Texture_Opacity})
  n38 = Subtract(A=Multiply(A=Multiply(A=@n28, B=@n37), B=Constant() {r=1}), B=Multiply(A=Constant() {r=1}, B=LinearInterpolate(Alpha=@n2)))
  n39 = ScalarParameter() {defaultValue=1.5, parameterName=Final_Opacity}
  MaterialFunctionCall(Alpha Threshold=Multiply(A=@n19.A, B=Multiply(A=@n38, B=@n39)), Random=Constant() {r=1}) {materialFunction=&DitherTemporalAA}
  Multiply(
   A=Multiply(A=@n23) {constB=2}
   B=MaterialFunctionCall(Exponent=ScalarParameter() {defaultValue=5, parameterName=AlignFresnel_Exp}) {materialFunction=&AxisAlignedFresnel}.Fresnel
  )
  Multiply(A=@n2, B=@n18.G)
  n40 = DynamicParameter() {defaultValue=<rgba 100 2500 1 1>, paramNames [WidthBase,WidthTip,Mult_Tile,Mult_Speed]}
  NamedRerouteDeclaration(None=@n40.Mult_Tile) {name=Mult_Tile, nodeColor=<rgba 0.3359 0 1 1>}
  NamedRerouteDeclaration(None=@n40.Mult_Speed) {name=Mult_Speed, nodeColor=<rgba 1 0 0.75 1>}
  NamedRerouteDeclaration(None=DynamicParameter() {parameterIndex=3, defaultValue=<rgba 1 1 2 2>, paramNames [Erode,Random,Param3,Param4]}.Random) {name=Random, nodeColor=<rgba 0 0.6719 1 1>}
  n41 = ComponentMask(None=TextureCoordinate()) {g=true}
  n42 = Time() {bIgnorePause=true, bOverride_Period=true, period=2}
  n43 = ComponentMask(None=TextureCoordinate()) {g=true}
  n44 = LinearInterpolate(
   A=Multiply(A=Frac(None=@n42), B=VectorParameter() {parameterName=PanSpeed, defaultValue=<rgba 250 1500 -500 1>}.RGB) {constB=250}
   B=Multiply(
    A=Time() {bIgnorePause=true, bOverride_Period=true, period=2}
    B=VectorParameter() {parameterName=PanSpeedStart, defaultValue=<rgba 250 1500 -500 1>}.RGB
   ) {constB=250}
   Alpha=Add(A=OneMinus(None=@n43), B=NamedRerouteUsage())
  )
  n45 = WorldPosition()
  n46 = Multiply(
   A=LinearInterpolate(A=ScalarParameter() {defaultValue=1, parameterName=s}, B=ScalarParameter() {defaultValue=1, parameterName=s2}, Alpha=OneMinus(None=@n41))
   B=Multiply(A=NamedRerouteUsage(), B=Add(A=@n44, B=@n45.XYZ))
  )
  n47 = Noise!(World Position=@n46, FilterWidth) {bTiling=true, bTurbulence=false, levelScale=1, levels=2, noiseFunction=NOISEFUNCTION_ValueALU, repeatSize=500, scale=0.001}
  n48 = OneMinus(None=ShadingPathSwitch(Default=Noise(World Position=@n46) {bTiling=true, bTurbulence=false, levels=2, repeatSize=500, scale=0.0001}, Forward=@n47, Mobile=@n47))
  n49 = ComponentMask(None=TextureCoordinate()) {g=true}
  n50 = LinearInterpolate(
   A=ScalarParameter() {defaultValue=50, parameterName=WobbleMult}
   B=ScalarParameter() {defaultValue=50, parameterName=WobbleMultStart}
   Alpha=@n49
  )
  NamedRerouteDeclaration(None=Multiply(A=VertexNormalWS(), B=Multiply(A=@n48, B=@n50) {constB=50})) {name=WPO, nodeColor=<rgba 0 1 0.1328 1>}
  NamedRerouteUsage()
  OneMinus(None=@n49)
  ScalarParameter() {parameterName=Erode}
 ]
}
