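# Translucent, two-sided unlit Niagara sprite material for blowing dune sand effects in the Helios Wasteland open desert.
# Color comes from a time-of-day tint multiplied by particle color, passed through FX emissive adjustments.
# Opacity takes a mip-biased smoke SubUV mask and adds panning tiled dust as grit that fades out with pixel depth.
# Depth fade, particle alpha and near-camera fade then soften that opacity.
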Material /Game/Environments/Helios/FX/OpenDesert/Materials/M_Wasteland_DuneSand
graph {
 properties {
  bAutomaticallySetUsageInEditor=false
  bUsedWithNiagaraSprites=true
  blendMode=BLEND_Translucent
  pixelDepthOffsetMode=PDOM_Legacy
  shadingModel=MSM_Unlit
  translucencyLightingMode=TLM_VolumetricPerVertexNonDirectional
  twoSided=true
  usedShadingModels=Unlit
 }
 nodes [
  n1 = VectorParameter() {defaultValue=<rgba 1 0.5749 0.4021 1>}
  MaterialFunctionCall(Emissive=Multiply(
    A=MaterialFunctionCall(Color_Morning=@n1.RGB, Color_Evening=@n1.RGB) {materialFunction=&Color_Time_of_Day}.Result
    B=ParticleColor().RGB
   )) {materialFunction=&MF_FX_EmissiveAdjustments}
  Multiply()
  n2 = TextureSample!(UVs, Tex, Bias=Constant() {r=2}, Apply View MipBias) {mipValueMode=TMVM_MipBias, samplerType=SAMPLERTYPE_LinearColor, texture=/Game/Effects/Fort_Effects/Textures/Smoke/T_Inky_Smoke_SubUV_8x8}
  n3 = Time()
  n4 = ScalarParameter() {defaultValue=0.1}
  n5 = TextureSample(UVs=Panner(Coordinate=TextureCoordinate() {uTiling=8, vTiling=8}, Time=Multiply(A=@n3, B=@n4)) {speedY=1}) {mipValueMode=TMVM_MipBias, texture=/Game/Global/Textures/Default/PointSplatter/T_Tiling_Dust}
  n6 = PixelDepth()
  n7 = ScalarParameter() {defaultValue=15000, parameterName=Grit DIstance}
  n8 = LinearInterpolate(A=ScalarParameter() {defaultValue=15, parameterName=Param_1}, Alpha=Saturate(None=Divide(A=@n6.R, B=@n7))) {constA=2, constB=0}
  n9 = Add(A=@n2.G, B=Multiply(A=@n2.G, B=Multiply(A=@n5.G, B=@n8)))
  Multiply(
   A=MaterialFunctionCall(Distance=ScalarParameter() {defaultValue=150, parameterName=Near Camera Fade}) {materialFunction=&Near_Camera_Fade}.Result
   B=Multiply(A=ParticleColor().A, B=DepthFade(Opacity=@n9) {fadeDistanceDefault=500})
  )
  Subtract()
 ]
}
